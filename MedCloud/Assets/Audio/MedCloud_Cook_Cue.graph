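# Sound cue for the MedCloud "cook" sound, routed to the gadgets sound class with weapon-pickup attenuation.
# A random pick of two one-shots, varied by a modulator, is mixed with a looping layer at half input volume.
# An enveloper fades that loop in to an early boosted peak, then eases it down over several seconds.
# A concurrency override limits simultaneous instances to a few.
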SoundCue /MedCloud/Assets/Audio/MedCloud_Cook_Cue
root=Mixer_0
nodes [
 {
  id=Enveloper_0
  children [WavePlayer_3]
  properties {pitchCurve {editorCurveData {keys [{time=0}]}}, volumeCurve {editorCurveData {keys [<at 0 0>,<at 0.5 2>,<at 1.1116 1.9217>,<at 5 1.0718>]}}}
 }
 {id=Mixer_0, children [Modulator_0,Enveloper_0], properties {inputVolume [1,0.5]}}
 {id=Modulator_0, children [Random_0], properties {pitchMin=0.9}}
 {id=Random_0, children [WavePlayer_0,WavePlayer_2], properties {bShouldExcludeFromBranchCulling=true, weights [1,1]}}
 {id=WavePlayer_0, properties {soundWaveAssetPtr=&/MedCloud_Cook_1}}
 {id=WavePlayer_2, properties {soundWaveAssetPtr=&/MedCloud_Cook_2}}
 {id=WavePlayer_3, properties {bLooping=true, soundWaveAssetPtr=&/MedCloud_Cook_Loop1}}
]
properties {
 attenuationSettings=&Fort_Weapon_Pickup_Att
 bOverrideConcurrency=true
 maxDistance=3500
 soundClassObject=&Fort_Gadgets
 volumeMultiplier=0.6
 concurrencyOverrides {maxCount=3}
}
