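# MedCloud_Equip_Cue is a Sound Cue that plays the equip sound for the MedCloud item.
# A Random node picks one of two equally weighted equip waves, excluded from branch culling, and feeds it through a Modulator at default settings.
# It borrows weapon reload attenuation, uses the loot collect sound class, and overrides concurrency to cap simultaneous instances at a few voices.
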SoundCue /MedCloud/Assets/Audio/MedCloud_Equip_Cue
root=Modulator_0
nodes [
 {id=Modulator_0, children [Random_0]}
 {id=Random_0, children [WavePlayer_0,WavePlayer_2], properties {bShouldExcludeFromBranchCulling=true, weights [1,1]}}
 {id=WavePlayer_0, properties {soundWaveAssetPtr=&/MedCloud_Equip_2}}
 {id=WavePlayer_2, properties {soundWaveAssetPtr=&/MedCloud_Equip_1}}
]
properties {
 attenuationSettings=&Fort_Weapon_Reload_Att
 bOverrideConcurrency=true
 maxDistance=2250
 soundClassObject=&Fort_Loot_Collect
 volumeMultiplier=0.8
 concurrencyOverrides {maxCount=3}
}
