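# Third-person looping heal sound cue for the MedCloud heal spray, classed as a consumable sound.
# A mixer layers a louder one-shot start sound, randomized in volume by a modulator, over a quieter looping heal wave.
# An enveloper fades the loop in over a second and slowly ramps its pitch up, so the tone climbs as healing continues.
# The cue uses the spray's shared attenuation and overrides concurrency to cap simultaneous instances.
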SoundCue /MedCloud/Assets/Audio/MedCloud_HealSpray_Heal_3P_Loop_Healing_Cue
root=Mixer_3
nodes [
 {
  id=Enveloper_3
  children [WavePlayer_14]
  properties {pitchCurve {editorCurveData {keys [<at 0 0.65>,<at 12 2>]}}, volumeCurve {editorCurveData {keys [<at 0 0>,<at 1 1>]}}}
 }
 {id=Mixer_3, children [Modulator_0,Enveloper_3], properties {inputVolume [1.5,0.5]}}
 {id=Modulator_0, children [WavePlayer_15], properties {volumeMax=0.8, volumeMin=0.6}}
 {id=WavePlayer_14, properties {bLooping=true, soundWaveAssetPtr=&/MedCloud_Player_Heal_Loop}}
 {id=WavePlayer_15, properties {soundWaveAssetPtr=&/MedCloud_Player_Heal_Loop_Start_01}}
]
properties {
 attenuationSettings=&/MedCloud_MedCloud_HealSpray_Att
 bOverrideConcurrency=true
 maxDistance=8900
 soundClassObject=&Fort_Consumables
 volumeMultiplier=1
 concurrencyOverrides {maxCount=3}
}
