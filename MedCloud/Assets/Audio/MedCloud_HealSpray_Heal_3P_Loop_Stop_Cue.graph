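# Sound cue for the MedCloud heal spray, named as the third-person heal loop stop sound.
# A modulator varies pitch on a single wave player, which plays the player heal loop start wave despite the Stop name.
# The cue uses the heal spray attenuation asset and consumables sound class, with a concurrency override capping simultaneous instances.
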SoundCue /MedCloud/Assets/Audio/MedCloud_HealSpray_Heal_3P_Loop_Stop_Cue
root=Modulator_0
nodes [
 {id=Modulator_0, children [WavePlayer_2], properties {pitchMax=1.15, volumeMax=1, volumeMin=1}}
 {id=WavePlayer_2, properties {soundWaveAssetPtr=&/MedCloud_Player_Heal_Loop_Start_01}}
]
properties {
 attenuationSettings=&/MedCloud_MedCloud_HealSpray_Att
 bOverrideConcurrency=true
 maxDistance=8900
 soundClassObject=&Fort_Consumables
 volumeMultiplier=1
 concurrencyOverrides {maxCount=3}
}
