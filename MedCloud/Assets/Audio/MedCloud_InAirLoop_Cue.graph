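# Sound cue for the MedCloud projectile in flight, layering a throw one-shot with a looping in-air sound.
# A mixer boosts the throw above the loop, and each branch has its own modulator and attenuation node.
# An enveloper fades the loop in from near silence.
# The throw branch overrides its attenuation with HRTF spatialization, distance filtering and a distant-reverb submix send, while the loop uses the shared in-air loop attenuation asset.
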SoundCue /MedCloud/Assets/Audio/MedCloud_InAirLoop_Cue
root=Mixer_0
nodes [
 {
  id=Attenuation_0
  children [Modulator_1]
  properties {
   attenuationSettings=/Game/Sounds/Attenuation/Fort_Weapon_Projectile_InAir_Launch_Throw_Att
   bOverrideAttenuation=true
   attenuationOverrides {
    bAttenuateWithLPF=true
    bEnableSubmixSends=true
    falloffDistance=20000
    hPFFrequencyAtMax=100
    lPFFrequencyAtMax=1500
    lPFRadiusMax=1250
    lPFRadiusMin=0
    reverbDistanceMax=1024
    reverbDistanceMin=0
    reverbWetLevelMax=0
    reverbWetLevelMin=1
    spatializationAlgorithm=SPATIALIZATION_HRTF
    stereoSpread=0
    attenuationShapeExtents {x=2600}
    submixSendSettings [{
      bEnableHPFCutoff=false
      bEnableLPFCutoff=false
      disableManualSendClamp=false
      hPFCutoff=20
      lPFCutoff=20000
      maxSendDistance=7500
      maxSendLevel=0.25
      minSendDistance=0
      minSendLevel=0
      sendLevel=0.2
      sendLevelControlMethod=Linear
      soundSubmix=&Fortnite_Reverb_Distant
      customSendLevelCurve {externalCurve=/Game/Sounds/SubMixes/Reverb/Fortnite_Reverb_Distant_Send}
     }]
   }
  }
 }
 {id=Attenuation_1, children [Enveloper_0], properties {attenuationSettings=/Game/Sounds/Attenuation/Fort_Weapon_Projectile_InAir_Loop_Att}}
 {
  id=Enveloper_0
  children [Modulator_0]
  properties {pitchCurve {editorCurveData {keys [{time=0}]}}, volumeCurve {editorCurveData {keys [<at 0.0001 0.0145>,<at 0.1225 0.0867>,<at 1 1>]}}}
 }
 {id=Mixer_0, children [Attenuation_0, Attenuation_1], properties {inputVolume [2.5,1]}}
 {id=Modulator_0, children [WavePlayer_0]}
 {id=Modulator_1, children [WavePlayer_1]}
 {id=WavePlayer_0, properties {bLooping=true, soundWaveAssetPtr=&/MedCloud_InAirLoop}}
 {id=WavePlayer_1, properties {soundWaveAssetPtr=&/MedCloud_Throw}}
]
properties {
 attenuationSettings=&Fort_Weapon_Projectile_InAir_Loop_Att
 maxDistance=22600
 attenuationOverrides {
  bAttenuateWithLPF=true
  bEnableSubmixSends=true
  falloffDistance=8000
  hPFFrequencyAtMax=100
  lPFFrequencyAtMax=1500
  lPFRadiusMax=1250
  lPFRadiusMin=0
  reverbDistanceMax=1024
  reverbDistanceMin=0
  reverbWetLevelMax=0
  reverbWetLevelMin=1
  spatializationAlgorithm=SPATIALIZATION_HRTF
  stereoSpread=0
  attenuationShapeExtents {x=2000}
  submixSendSettings [{
    bEnableHPFCutoff=false
    bEnableLPFCutoff=false
    disableManualSendClamp=false
    hPFCutoff=20
    lPFCutoff=20000
    maxSendDistance=7500
    maxSendLevel=0.25
    minSendDistance=0
    minSendLevel=0
    sendLevel=0.2
    sendLevelControlMethod=Linear
    soundSubmix=&Fortnite_Reverb_Distant
    customSendLevelCurve {externalCurve=&Fortnite_Reverb_Distant_Send}
   }]
 }
}
