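# Looping MedCloud sound cue in the gadgets sound class, layering an inner and an outer loop.
# Each loop has its own attenuation node: the inner uses inline short-falloff, low-pass overrides, the outer a dedicated MedCloud attenuation asset.
# An equal-weight mixer feeds a fixed-volume modulator at the root.
# Concurrency is overridden to cap simultaneous instances; cue-level attenuation values exist, but their override flag is unset.
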SoundCue /MedCloud/Assets/Audio/MedCloud_Loop_Cue
root=Modulator_0
nodes [
 {
  id=Attenuation_0
  children [WavePlayer_1]
  properties {
   attenuationSettings=/Game/Sounds/Attenuation/Fort_Player_Swimming_Athena_Att
   bOverrideAttenuation=true
   attenuationOverrides {bAttenuateWithLPF=true, falloffDistance=500, attenuationShapeExtents {x=0}}
  }
 }
 {id=Attenuation_1, children [WavePlayer_2], properties {attenuationSettings=&/ATT_MedCloud_Loop}}
 {id=Mixer_0, children [Attenuation_0, Attenuation_1], properties {inputVolume [0.3,0.3]}}
 {id=Modulator_0, children [Mixer_0], properties {volumeMax=0.5, volumeMin=0.5}}
 {id=WavePlayer_1, properties {bLooping=true, soundWaveAssetPtr=&/MedCloud_InnerLoop}}
 {id=WavePlayer_2, properties {bLooping=true, soundWaveAssetPtr=&/MedCloud_OuterLoop}}
]
properties {
 bOverrideConcurrency=true
 maxDistance=8750
 pitchMultiplier=1.1
 soundClassObject=&Fort_Gadgets
 volumeMultiplier=1.5
 attenuationOverrides {
  bEnableOcclusion=true
  distanceAlgorithm=Logarithmic
  falloffDistance=1800
  nonSpatializedRadiusStart=600
  occlusionLowPassFilterFrequency=18000
  occlusionTraceChannel=ECC_GameTraceChannel11
  occlusionVolumeAttenuation=0.6
  stereoSpread=2000
 }
 concurrencyOverrides {maxCount=3}
}
