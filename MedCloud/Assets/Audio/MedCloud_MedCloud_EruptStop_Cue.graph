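# Sound cue for the MedCloud shield sprinkler's erupt-stop sound, assigned to the consumables sound class.
# A Random node picks one of two equally weighted wave variations and feeds a Modulator at the root.
# Inline attenuation overrides set distance falloff, stereo spread and occlusion with volume and low-pass filtering.
# A concurrency override caps simultaneous instances at a few, stopping the oldest first.
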SoundCue /MedCloud/Assets/Audio/MedCloud_MedCloud_EruptStop_Cue
root=Modulator_0
nodes [
 {id=Modulator_0, children [Random_0]}
 {id=Random_0, children [WavePlayer_3,WavePlayer_4], properties {weights [1,1]}}
 {id=WavePlayer_3, properties {soundWaveAssetPtr=&/MedCloud_ShieldSprinkler_EruptStop_01}}
 {id=WavePlayer_4, properties {soundWaveAssetPtr=&/MedCloud_ShieldSprinkler_EruptStop_02}}
]
properties {
 attenuationSettings=/Game/Sounds/Attenuation/Fort_Bullet_Impact_NPC_Weapon_Athena_Att
 bOverrideAttenuation=true
 bOverrideConcurrency=true
 maxDistance=1750
 soundClassObject=&Fort_Consumables
 volumeMultiplier=1.1
 attenuationOverrides {
  bEnableOcclusion=true
  falloffDistance=1700
  occlusionLowPassFilterFrequency=18000
  occlusionTraceChannel=ECC_GameTraceChannel11
  occlusionVolumeAttenuation=0.8
  stereoSpread=2000
  attenuationShapeExtents {x=50}
 }
 concurrencyOverrides {maxCount=3, resolutionRule=StopOldest}
}
