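# MedCloud_Pickup_Cue is a Sound Cue that plays when the MedCloud item is picked up.
# A Random node picks one of two equally weighted pickup waves, then passes it through a Modulator for playback variation.
# It uses the weapon pickup attenuation and loot collect sound class, with slightly reduced volume.
# A concurrency override caps the number of simultaneous instances.
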SoundCue /MedCloud/Assets/Audio/MedCloud_Pickup_Cue
root=Modulator_0
nodes [
 {id=Modulator_0, children [Random_0]}
 {id=Random_0, children [WavePlayer_2,WavePlayer_0], properties {bShouldExcludeFromBranchCulling=true, weights [1,1]}}
 {id=WavePlayer_0, properties {soundWaveAssetPtr=/MedCloud/MedCloud_PickUp_2}}
 {id=WavePlayer_2, properties {soundWaveAssetPtr=/MedCloud/MedCloud_PickUp_1}}
]
properties {
 attenuationSettings=&Fort_Weapon_Pickup_Att
 bOverrideConcurrency=true
 maxDistance=3500
 soundClassObject=&Fort_Loot_Collect
 volumeMultiplier=0.7
 concurrencyOverrides {maxCount=3}
}
