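# Niagara effect for the MedCloud grenade's dome: a green-tinted smoke volume sized by FX_Scale, with dome lifetime set by FXAge.
# A burst-spawned dome mesh with curve-animated material layers is filled by orbiting mist sprites, swirling motes and spinning wind-gust meshes.
# System FadeOut and timeline values dissolve the layers, while time-of-day color and NightMultiplier adjust smoke brightness.
# Mist also brightens by sky light direction, and custom modules fade sprites relative to the camera.
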NiagaraSystem /MedCloud/Assets/FX/NS_MedCloud_Grenade_Dome
dependencies {
 usedDataInterfaces [
  &NiagaraDataInterfaceCurve
  &NiagaraDataInterfaceCamera
  &NiagaraDataInterfaceSpriteRendererInfo
  &NiagaraDataInterfaceMeshRendererInfo
  &NiagaraDataInterfaceArrayFloat
 ]
 usedDynamicInputs [&Divide_Float]
 usedModules [
  &SystemState
  &Timeline
  &self:SystemSpawnScript.SystemScriptSource.SystemScriptGraph.NiagaraNodeAssignment_1.SetVariables_#1
  &self:SystemSpawnScript.SystemScriptSource.SystemScriptGraph.NiagaraNodeAssignment_2.SetVariables_#2
 ]
}
emitters {
 Dome {
  emitterUpdateScript [
   &EmitterState(
    Inactive Response=::NewEnumerator1
    Life Cycle Mode=::NewEnumerator1
    Loop Behavior=::NewEnumerator1
    Loop Duration Mode=::NewEnumerator1
    UseLoopDelay=false
   )
   &SpawnBurst_Instantaneous(Use Spawn Probability=false, UseLoopCountLimit=false)
  ]
  particleSpawnScript [
   &InitializeParticle(
    Color=&MultiplyLinearColor(
     Linear Color=<rgba 0.0138 1 0.0017 0.2>
     Scale Factor=&MakeVector4_FromLinearColor(LinearColor=$User.SmokeColor)
    )
    Lifetime=$User.FXAge
    Mass Mode=::NewEnumerator3
    Mesh Scale Mode=::NewEnumerator3
    Mesh Uniform Scale=&Multiply_Float(A=7, B=$User.FX_Scale)
    Sprite Rotation Mode=::NewEnumerator2
    UsePositionOffset=false
    Write Material Random=true
   )
   /SetVariables(Particles.CameraOffsetCore=-120) {isSetParameters=true}
   /SetVariables(Particles.MeshOrientation=$System.DomeOrientation) {isSetParameters=true}
  ]
  particleUpdateScript [
   &ParticleState(Kill Particles When Lifetime Has Elapsed=false, Let Infinitely Lived Particles Die When Emitter Deactivates=true)
   &ScaleColor(
    Scale Alpha=$System.FadeOut
    Scale RGB=&Multiply_Vector(
     A=&MakeVectorFromLinearColorRGB(Color=$System.TODColorMult)
     B=&VectorFromFloat(Value=$User.NightMultiplier)
    )
    ScaleA=true
    ScaleRGB=true
   )
   &DynamicMaterialParameters(
    Index 0 Param 1=1
    Index 0 Param 2=6
    Index 0 Param 3=&FloatFromCurve(FloatCurve=DataInterface() {propertyValues {Curve {keys [
         {arriveTangent=-0.9649, interpMode=RCIM_Cubic, leaveTangent=-0.9649, tangentMode=RCTM_User, time=0, value=1}
         {arriveTangent=-6.2582, leaveTangent=-1.25, tangentMode=RCTM_User, time=0.03, value=0}
         {arriveTangent=3.1809, interpMode=RCIM_Cubic, leaveTangent=3.1809, tangentMode=RCTM_User, time=0.9, value=0}
         {arriveTangent=10.4506, interpMode=RCIM_Cubic, leaveTangent=10.4506, tangentMode=RCTM_User, time=1, value=1}
        ]}}})
    Index 0 Param 4=&Multiply_Float(A=$User.FX_Scale, B=5)
    Module.Index0Param0Write=true
    Module.Index0Param1Write=true
    Module.Index0Param2Write=true
    Module.Index0Param3Write=true
   )
   /SetVariables(Particles.DynMatPar_Inner:Vector4f=&MakeVector4(
     X=1
     Y=13
     Z=&FloatFromCurve(FloatCurve=DataInterface() {propertyValues {Curve {keys [
          {arriveTangent=-17.4004, interpMode=RCIM_Cubic, leaveTangent=-17.4004, tangentMode=RCTM_User, time=0, value=1}
          {arriveTangent=-0.0714, tangentMode=RCTM_User, time=0.1, value=0}
          {arriveTangent=0.6539, interpMode=RCIM_Cubic, leaveTangent=0.6539, tangentMode=RCTM_User, time=0.8, value=0}
          {arriveTangent=1.435, interpMode=RCIM_Cubic, leaveTangent=1.435, tangentMode=RCTM_User, time=1, value=1}
         ]}}})
     W=&Multiply_Float(A=$User.FX_Scale, B=2)
    )) {isSetParameters=true}
   /SetVariables(Particles.DynMatPar_SnowThick:Vector4f=&MakeVector4(
     X=1
     Y=3
     Z=&FloatFromCurve(FloatCurve=DataInterface() {propertyValues {Curve {keys [
          {arriveTangent=-4.8072, interpMode=RCIM_Cubic, leaveTangent=-4.8072, tangentMode=RCTM_User, time=0, value=1}
          {arriveTangent=-0.2555, interpMode=RCIM_Constant, tangentMode=RCTM_User, time=0.15, value=0}
          {arriveTangentWeight=0.2033, interpMode=RCIM_Cubic, leaveTangentWeight=0.0467, tangentMode=RCTM_Break, time=0.8, value=0}
          {interpMode=RCIM_Cubic, tangentMode=RCTM_User, time=1, value=1}
         ]}}})
     W=&Multiply_Float(A=$User.FX_Scale, B=1)
    )) {isSetParameters=true}
   /SetVariables(Particles.ColorSnowThick:LinearColor=&MultiplyLinearColor(
     Linear Color=$User.SmokeColor
     Scale Factor=&MakeFloat4FromVectorAndFloat(
      "Vector (RGB)"=&VectorFromFloat(Value=$User.NightMultiplier)
      "Float (Alpha)"=&Multiply_Float(
       A=&FloatFromCurve(
        FloatCurve=DataInterface() {propertyValues {Curve {keys [
            {interpMode=RCIM_Cubic, tangentMode=RCTM_User, time=0, value=0}
            {interpMode=RCIM_Cubic, tangentMode=RCTM_User, time=0.1, value=1}
            {arriveTangent=-0.0025, interpMode=RCIM_Cubic, leaveTangent=-0.0025, tangentMode=RCTM_User, time=0.9, value=1}
            <at 0.92 0 RCIM_Cubic>
           ]}}}
        Scale Curve=&Multiply_Float(
         A=0.5
         B=&Multiply_Float(A=$System.FadeOut, B=&OneMinusFloat(Float=$System.Timeline.NormalizedTime))
        )
       )
       B=1
      )
     )
    )) {isSetParameters=true}
   /SetVariables(Particles.ColorCore:LinearColor=&MultiplyLinearColor(
     Linear Color=$Particles.Color
     Scale Factor=&MakeFloat4FromVectorAndFloat("Vector (RGB)"=&VectorFromFloat(Value=$User.NightMultiplier), "Float (Alpha)"=1)
    )) {isSetParameters=true}
   /SetVariables(Particles.ScaleThickSnow:Vector3f=&Multiply_VectorByFloat(Vector=$Particles.Scale, Float=1)) {isSetParameters=true}
  ]
  rendererClasses [&NiagaraMeshRendererProperties]
  renderers [{rendererClass=&NiagaraMeshRendererProperties, rendererIndex=0}]
 }
 Mist {
  emitterUpdateScript [
   &EmitterState()
   &SpawnRate(SpawnRate=25, Use Spawn Probability=false)
  ]
  particleSpawnScript [
   /SetVariables(Particles.SmokeGlobeAlpha:NiagaraFloat=&FloatFromCurve(
     FloatCurve=DataInterface() {propertyValues {Curve {keys [
         {arriveTangent=0.3467, interpMode=RCIM_Cubic, leaveTangent=0.3467, tangentMode=RCTM_User, time=0, value=0}
         {arriveTangent=0.4113, interpMode=RCIM_Cubic, leaveTangent=0.4113, tangentMode=RCTM_User, time=1, value=0.2}
        ]}}}
     CurveIndex=&RandomRangeFloat(Minimum=0, Maximum=1)
    )) {isSetParameters=true}
   &InitializeParticle(
    Color=&MultiplyLinearColor(
     Linear Color=&MultiplyLinearColor(
      Linear Color=&RandomRangeLinearColor(Minimum=<rgba 0.5521 0.5521 0.5521 1>, Maximum=<rgba 0.8 0.8 0.8 1>)
      Scale Factor=&MakeVector4_FromLinearColor(LinearColor=<rgba 0.4537 1 0.7838 1>)
     )
     Scale Factor=&MakeVector4_FromLinearColor(LinearColor=$User.SmokeColor)
    )
    Lifetime Max=2.4
    Lifetime Min=1.5
    Lifetime Mode=::NewEnumerator1
    Mass Mode=::NewEnumerator3
    Position Offset=&Multiply_VectorByFloat(Vector=<0 0 0>, Float=$User.FX_Scale)
    Sprite Rotation Mode=::NewEnumerator2
    Sprite Size Mode=::NewEnumerator3
    Uniform Sprite Size=&Multiply_Float(A=$User.FX_Scale, B=&RandomRangeFloat(Minimum=300, Maximum=350))
    UsePositionOffset=true
    Write Material Random=true
   )
   /SetVariables(Particles.InitialOffset:NiagaraFloat=&Multiply_Float(A=$System.Age, B=1)) {isSetParameters=true}
   &CameraOffset(Camera Offset Amount=&Multiply_Float(A=60, B=$User.FX_Scale))
  ]
  particleUpdateScript [
   &ParticleState()
   &RotateAroundPoint(
    Radius:NiagaraFloat=&Multiply_Float(
     A=&Lerp_Float(
      A=350
      B=1
      Alpha=&FloatFromCurve(
       FloatCurve=DataInterface() {propertyValues {Curve {keys [
           {arriveTangent=0.1398, interpMode=RCIM_Cubic, leaveTangent=0.1398, tangentMode=RCTM_User, time=0, value=0}
           {arriveTangent=3.3869, interpMode=RCIM_Cubic, leaveTangent=3.3869, tangentMode=RCTM_User, time=1, value=1}
          ]}}}
       CurveIndex=$Particles.SmokeGlobeAlpha
      )
     )
     B=$User.FX_Scale
    )
    Rotation Center:NiagaraPosition=&Multiply_VectorByFloat(
     Vector=&Lerp_Vector(
      A=<0 0 0>
      B=<0 0 350>
      Alpha=&Multiply_Float(
       A=$Particles.SmokeGlobeAlpha
       B=&UniformAOrBFloat(A=-1, Evaluation Type=::NewEnumerator0, Recalculate Random Each Loop=true, Override Seed=false)
      )
     )
     Float=$User.FX_Scale
    )
    Rotation Phase:NiagaraFloat=&Add_Float(
     A=&FloatFromCurve(
      FloatCurve=DataInterface() {propertyValues {Curve {keys [<at 0 1>,<at 1 0>]}}}
      Scale Curve=&RandomRangeFloat(Minimum=-0.25, Maximum=-0.25)
     )
     B=$Particles.InitialOffset
    )
   )
   &self:FadeOutBackParticles(
    CamInsideFallof=100
    CamInsideRadius=600
    CameraForwardVector:Vector3f=&ReturnCameraVector(Camera Vectors=::NewEnumerator1, Coordinate Space=::World)
    CameraPosition:NiagaraPosition=&ReturnCameraPosition(Coordinate Space:ENiagaraCoordinateSpace=::World, Camera Query:NiagaraDataInterfaceCamera=DataInterface())
    FadeOutSoftness=0.15
    FadeOutTreshold=0.5
   ) {module=FadeOutBackParticles}
   &self:FadeOutWhenInSmoke(AlphaMax=1, AlphaMin=0.5, FadeFallof=100, FadeRadius=600, SmokePos:NiagaraPosition=$Engine.Owner.Position) {module=FadeOutWhenInSmoke}
   &ScaleSpriteSize(Uniform Curve Sprite Scale=DataInterface() {propertyValues {Curve {keys [{interpMode=RCIM_Cubic, tangentMode=RCTM_User, time=0, value=0.6}, {interpMode=RCIM_Cubic, tangentMode=RCTM_User, time=0.7, value=1}]}}})
   &ScaleColor(
    Scale Alpha=&FloatFromCurve(
     FloatCurve=DataInterface() {propertyValues {Curve {keys [
         {interpMode=RCIM_Cubic, tangentMode=RCTM_User, time=0, value=0}
         {interpMode=RCIM_Cubic, tangentMode=RCTM_User, time=0.4, value=1}
         {interpMode=RCIM_Cubic, tangentMode=RCTM_User, time=1, value=0}
        ]}}}
     Scale Curve=&Multiply_Float(
      A=&Multiply_Float(
       A=&Multiply_Float(A=$System.FadeOut, B=&OneMinusFloat(Float=$System.Timeline.NormalizedTime))
       B=$Output.FadeOutBackParticles.FresnelAlpha
      )
      B=$Output.FadeOutWhenInSmoke.FadeOutAlpha
     )
    )
    Scale RGB=&Multiply_Vector(
     A=&MakeVectorFromLinearColorRGB(Color=$System.TODColorMult)
     B=&VectorFromFloat(Value=&Multiply_Float(
       A=$User.NightMultiplier
       B=&Lerp_Float(
        A=1
        B=1.5
        Alpha=&FloatFromCurve(
         FloatCurve=DataInterface() {propertyValues {Curve {keys [<at -0.75 0>,<at 1 1>]}}}
         CurveIndex=&DotProduct(
          Vector A=&NormalizeVector(
           Vector To Normalize=&MakeVectorFromLinearColorRGB(Color=$NPC.FortniteNPC.SkyAtmosphereLightDirection)
           Safe Normalize=true
           Safe Fallback Threshold=0
           Safe Fallback Vector=<0 0 0.001>
          )
          Vector B=&TransformVector(
           Vector=&NormalizeVector(
            Vector To Normalize=$Particles.RotateAroundPoint.VectorFromCenter
            Safe Normalize=true
            Safe Fallback Threshold=0
            Safe Fallback Vector=<0 0 0.001>
           )
           Destination Space=::World
          )
         )
        )
       )
      ))
    )
    ScaleA=true
    ScaleRGB=true
   )
   &SubUVAnimation(
    "Sprite Renderer "=DataInterface() {propertyValues {SpriteRenderer=&self:Mist_1.NiagaraSpriteRendererProperties_0}}
    UseEndFrame=false
    UseStartFrame=false
   )
   &Drag(Drag=4, Rotational Drag=2, Use Linear Drag=true, Use Rotational Drag=true)
   &SolveForcesAndVelocity(Clamp Velocity=false, Limit Acceleration=false)
   &DynamicMaterialParameters(Module.Index0Param0Write=true, Module.Index0Param1Write=true, Module.Index0Param2Write=true, Module.Index0Param3Write=true)
  ]
  rendererClasses [&NiagaraSpriteRendererProperties]
  renderers [{rendererClass=&NiagaraSpriteRendererProperties}]
 }
 Motes {
  emitterUpdateScript [
   &EmitterState()
   &SpawnRate(
    Spawn Probability=&FloatFromCurve(
     FloatCurve=DataInterface() {propertyValues {Curve {keys [
         {interpMode=RCIM_Cubic, tangentMode=RCTM_User, time=0.1, value=0}
         {interpMode=RCIM_Cubic, tangentMode=RCTM_User, time=1, value=1}
         {arriveTangent=0.0061, interpMode=RCIM_Cubic, leaveTangent=-0.1111, tangentMode=RCTM_Break, time=9, value=1}
         <at 10.6 0 RCIM_Cubic>
        ]}}}
     CurveIndex=$Emitter.Age
    )
    SpawnRate=20
    Use Spawn Probability=false
   )
   &SpawnRate(
    Spawn Probability=&FloatFromCurve(
     FloatCurve=DataInterface() {propertyValues {Curve {keys [
         {interpMode=RCIM_Cubic, tangentMode=RCTM_User, time=0.1, value=0}
         {interpMode=RCIM_Cubic, tangentMode=RCTM_User, time=1, value=1}
         {arriveTangent=0.0061, interpMode=RCIM_Cubic, leaveTangent=-0.1111, tangentMode=RCTM_Break, time=9, value=1}
         <at 10.6 0 RCIM_Cubic>
        ]}}}
     CurveIndex=$Emitter.Age
    )
    SpawnGroup=1
    SpawnRate=30
    Use Spawn Probability=false
   ) {module=SpawnRate001}
  ]
  particleSpawnScript [
   /SetVariables(
    Particles.Particle_Depth:NiagaraFloat=&FloatFromCurve(
     FloatCurve=DataInterface() {propertyValues {Curve {keys [
         {arriveTangent=1.2753, interpMode=RCIM_Cubic, leaveTangent=1.2753, tangentMode=RCTM_User, time=0, value=0}
         {arriveTangent=0.204, interpMode=RCIM_Cubic, leaveTangent=0.204, tangentMode=RCTM_User, time=0.9, value=1}
        ]}}}
     CurveIndex=&RandomRangeFloat(Minimum=0, Maximum=1)
    )
    Particles.Particle_Height:NiagaraFloat=&RandomRangeFloat(Minimum=0, Maximum=1)
    Particles.Particle_Scale:NiagaraFloat=&RandomRangeFloat(Minimum=0, Maximum=1)
   ) {isSetParameters=true}
   &PartitionParticles()
   /SetVariables(Particles.VisibilityTag=&SpawnGroup()) {isSetParameters=true}
   &InitializeParticle(
    Color Maximum=<rgba 0 1 0.3727 1>
    Color Minimum=<rgba 0 0.5833 0.2061 1>
    Color Mode=::NewEnumerator2
    Lifetime Max=1.5
    Lifetime Min=0.7
    Lifetime Mode=::NewEnumerator1
    Mass Mode=::NewEnumerator3
    Sprite Size Mode=::NewEnumerator3
    Uniform Sprite Size=&Multiply_Float(A=$User.FX_Scale, B=&Lerp_Float(A=12, B=27, Alpha=$Particles.Particle_Scale))
    UsePositionOffset=false
    Write Material Random=true
   )
   &ShapeLocation(
    MaskSpawnGroup=false
    Sphere Distribution=::NewEnumerator1
    Sphere Radius=&Multiply_Float(A=450, B=$User.FX_Scale)
    UseOwnerScale=true
    V Position=&RandomRangeFloat(Minimum=0.5, Maximum=1)
   )
   &CameraOffset(Camera Offset Amount=6)
   /SetVariables(Particles.RotationDirection:Vector3f=&Lerp_Vector(
     A=&RandomRangeVector(Minimum=<-0.6 -0.6 1>, Maximum=<0.8 0.8 1>)
     B=&RandomRangeVector(Minimum=<-0.2 -0.2 1>, Maximum=<0.3 0.3 1>)
     Alpha=&FloatFromCurve(
      FloatCurve=DataInterface() {propertyValues {Curve {keys [
          {interpMode=RCIM_Cubic, tangentMode=RCTM_User, time=0, value=0}
          {arriveTangent=0.5342, interpMode=RCIM_Cubic, leaveTangent=0.5342, tangentMode=RCTM_User, time=1, value=1}
         ]}}}
      CurveIndex=$Particles.Particle_Depth
     )
    )) {isSetParameters=true}
   &DynamicMaterialParameters(
    Index 0 Param 1=&RandomRangeFloat(Minimum=1, Maximum=2)
    Module.Index0Param0Write=true
    Module.Index0Param1Write=false
    Module.Index0Param2Write=false
    Module.Index0Param3Write=false
   )
  ]
  particleUpdateScript [
   &ParticleState()
   &self:FadeOutBackParticles(
    CamInsideFallof=100
    CamInsideRadius=600
    CameraForwardVector:Vector3f=&ReturnCameraVector(Camera Vectors=::NewEnumerator1, Coordinate Space=::World)
    CameraPosition:NiagaraPosition=&ReturnCameraPosition(Coordinate Space:ENiagaraCoordinateSpace=::World, Camera Query:NiagaraDataInterfaceCamera=DataInterface())
    FadeOutSoftness=0.15
    FadeOutTreshold=0.5
   ) {module=FadeOutBackParticles}
   &ScaleSpriteSize(
    Uniform Curve Scale=&Multiply_Float(
     A=&Multiply_Float(A=$System.FadeOut, B=&OneMinusFloat(Float=$System.Timeline.NormalizedTime))
     B=1
    )
    Uniform Curve Sprite Scale=DataInterface() {propertyValues {Curve {keys [
        {arriveTangent=8.1227, interpMode=RCIM_Cubic, leaveTangent=8.1227, tangentMode=RCTM_User, time=0, value=0.1}
        <at 0.25 1 RCIM_Cubic>
        {arriveTangent=-3.9188, interpMode=RCIM_Cubic, leaveTangent=-3.9188, tangentMode=RCTM_User, time=1, value=0.1}
       ]}}}
   )
   &SolveForcesAndVelocity(Clamp Velocity=false, Limit Acceleration=false)
   &RotateAroundPoint(
    Derive the Rotation Axis=false
    Radius:NiagaraFloat=&Multiply_Float(
     A=$User.FX_Scale
     B=&Multiply_Float(
      A=&RandomRangeFloat(Minimum=0.9, Maximum=1)
      B=&Lerp_Float(
       A=0
       B=400
       Alpha=&FloatFromCurve(
        FloatCurve=DataInterface() {propertyValues {Curve {keys [
            {arriveTangent=4.631, interpMode=RCIM_Cubic, leaveTangent=4.631, tangentMode=RCTM_User, time=0, value=0}
            {interpMode=RCIM_Cubic, tangentMode=RCTM_User, time=0.5, value=1}
            {arriveTangent=-4.4542, interpMode=RCIM_Cubic, leaveTangent=-4.4542, tangentMode=RCTM_User, time=1, value=0}
           ]}}}
        CurveIndex=$Particles.Particle_Height
       )
      )
     )
    )
    Rotation Axis:Vector3f=$Particles.RotationDirection
    Rotation Center:NiagaraPosition=&Multiply_VectorByFloat(
     Vector=&Lerp_Vector(
      A=<0 0 0>
      B=<0 0 250>
      Alpha=&Multiply_Float(
       A=$Particles.Particle_Height
       B=&UniformAOrBFloat(A=-1, Evaluation Type=::NewEnumerator0, Recalculate Random Each Loop=true, Override Seed=false)
      )
     )
     Float=$User.FX_Scale
    )
    Rotation Phase:NiagaraFloat=&RandomRangeFloat(Minimum=-1, Maximum=1)
    Rotation Rate:NiagaraFloat=&Multiply_Float(
     A=&FloatFromCurve(FloatCurve=DataInterface() {propertyValues {Curve {keys [
          {interpMode=RCIM_Cubic, tangentMode=RCTM_User, time=0, value=0.2}
          {arriveTangent=0.3, interpMode=RCIM_Cubic, leaveTangent=0.3, time=0.3, value=1}
          {interpMode=RCIM_Cubic, tangentMode=RCTM_User, time=1, value=0.5}
         ]}}})
     B=&Multiply_Float(
      A=&Lerp_Float(A=0.6, B=1.3, Alpha=$Particles.Particle_Scale)
      B=&Multiply_Float(
       A=&Lerp_Float(A=0, B=0.2, Alpha=$Particles.Particle_Depth)
       B=&Lerp_Float(
        A=1
        B=1
        Alpha=&FloatFromCurve(
         FloatCurve=DataInterface() {propertyValues {Curve {keys [
             {arriveTangent=5.5061, interpMode=RCIM_Cubic, leaveTangent=5.5061, tangentMode=RCTM_User, time=0, value=0}
             {interpMode=RCIM_Cubic, tangentMode=RCTM_User, time=0.5, value=1}
             {arriveTangent=-6.0544, interpMode=RCIM_Cubic, leaveTangent=-6.0544, tangentMode=RCTM_User, time=1, value=0}
            ]}}}
         CurveIndex=$Particles.Particle_Height
        )
       )
      )
     )
    )
   )
   &ScaleColor(
    Scale Alpha=&Lerp_Float(A=0.5, B=2, Alpha=$Output.FadeOutBackParticles.FresnelAlpha)
    Scale RGB=&VectorFromFloat(Value=&RandomRangeFloat(Minimum=0.5, Maximum=2))
    ScaleA=true
    ScaleRGB=true
   )
  ]
  rendererClasses [&NiagaraSpriteRendererProperties]
  renderers [
   {rendererClass=&NiagaraSpriteRendererProperties}
   {rendererClass=&NiagaraSpriteRendererProperties, rendererIndex=1}
  ]
 }
 WindGusts {
  emitterUpdateScript [
   &EmitterState()
   &SpawnRate(
    Spawn Probability=&FloatFromCurve(
     FloatCurve=DataInterface() {propertyValues {Curve {keys [
         {interpMode=RCIM_Cubic, tangentMode=RCTM_User, time=0, value=0}
         {interpMode=RCIM_Cubic, tangentMode=RCTM_User, time=0.1, value=1}
         {arriveTangent=-4.1541, interpMode=RCIM_Cubic, leaveTangent=-4.1541, tangentMode=RCTM_User, time=1, value=0}
        ]}}}
     CurveIndex=$System.NormalizedLoopAge
    )
    SpawnRate=12
    Use Spawn Probability=false
   )
  ]
  particleSpawnScript [
   /SetVariables(Particles.Particle_Height:NiagaraFloat=&RandomRangeFloat(Minimum=0.4, Maximum=0.6)) {isSetParameters=true}
   &InitializeParticle(
    Color=&MultiplyLinearColor(
     Linear Color=&MultiplyLinearColor(
      Linear Color=&RandomRangeLinearColor(Minimum=<rgba 0.7448 0.7448 0.7448 1>, Maximum=<rgba 0.7188 0.7188 0.7188 1>)
      Scale Factor=&MakeVector4_FromLinearColor(LinearColor=<rgba 1 1 1 1>)
     )
     Scale Factor=&MakeVector4_FromLinearColor(LinearColor=$User.SmokeColor)
    )
    Lifetime Max=1.2
    Lifetime Min=0.9
    Lifetime Mode=::NewEnumerator1
    Mass Mode=::NewEnumerator3
    Mesh Scale=&Multiply_VectorByFloat(
     Vector=&RandomRangeVector(Minimum=<1 1 0.5>, Maximum=<1 1 1.5>)
     Float=&Multiply_Float(
      A=$User.FX_Scale
      B=&Lerp_Float(
       A=5.5
       B=6
       Alpha=&FloatFromCurve(
        FloatCurve=DataInterface() {propertyValues {Curve {keys [
            {arriveTangent=5.7133, interpMode=RCIM_Cubic, leaveTangent=5.7133, tangentMode=RCTM_User, time=0, value=0}
            {interpMode=RCIM_Cubic, tangentMode=RCTM_User, time=0.5, value=1}
            {arriveTangent=-6.0152, interpMode=RCIM_Cubic, leaveTangent=-6.0152, tangentMode=RCTM_User, time=1, value=0}
           ]}}}
        CurveIndex=$Particles.Particle_Height
        Scale Curve=&RandomRangeFloat(Minimum=0.9, Maximum=1)
       )
      )
     )
    )
    Mesh Scale Mode=::NewEnumerator4
    Sprite Rotation Mode=::NewEnumerator2
    UsePositionOffset=false
    Write Material Random=true
   )
   &SystemLocation(
    MaskSpawnGroup=false
    Offset=&MakeVector(
     X=0
     Y=0
     Z=&Multiply_Float(A=$User.FX_Scale, B=&Lerp_Float(A=-300, B=300, Alpha=$Particles.Particle_Height))
    )
   )
   &InitialMeshOrientation(
    Mesh Orientation Mode=::NewEnumerator2
    Orientation Coordinate Space=::World
    Rotation=&RandomRangeVector(Minimum=<-0.04 -0.04 -1>, Maximum=<0 0 1>)
    Use Rotation Vector=true
   )
   &CameraOffset(Camera Offset Amount=15)
  ]
  particleUpdateScript [
   &ParticleState()
   &ScaleMeshSize(Scale Factor=&Multiply_VectorByFloat(
     Vector=&MakeVector(
      X=1
      Y=1
      Z=&SelectFloatFromArray(
       Array Sampling Mode=::NewEnumerator0
       Direct Array Index=?
       Float Selection Array=DataInterface() {propertyValues {FloatData [1,-1]}}
       Random Seed=0
      )
     )
     Float=&Lerp_Float(
      A=0.8
      B=1.2
      Alpha=&FloatFromCurve(FloatCurve=DataInterface() {propertyValues {Curve {keys [
           {arriveTangent=2.1221, interpMode=RCIM_Cubic, leaveTangent=2.1221, tangentMode=RCTM_User, time=0, value=0}
           {interpMode=RCIM_Cubic, tangentMode=RCTM_User, time=0.86, value=1}
          ]}}})
     )
    ))
   &UpdateMeshOrientation(Rotation Rate=&RandomRangeFloat(Minimum=0.1, Maximum=0.3), Rotation Vector=<0 0 1>)
   &ScaleColor(
    Scale Alpha=&FloatFromCurve(
     FloatCurve=DataInterface() {propertyValues {Curve {keys [
         {interpMode=RCIM_Cubic, leaveTangentWeight=1.0516, tangentMode=RCTM_User, time=0, value=0}
         {arriveTangentWeight=1.164, interpMode=RCIM_Cubic, leaveTangentWeight=1.164, tangentMode=RCTM_User, time=0.5, value=1}
         {arriveTangent=0.4048, interpMode=RCIM_Cubic, leaveTangent=0.4048, tangentMode=RCTM_User, time=1, value=0}
        ]}}}
     Scale Curve=&Multiply_Float(
      A=&RandomRangeFloat(Minimum=2, Maximum=5)
      B=&Multiply_Float(
       A=&Multiply_Float(A=$System.FadeOut, B=&OneMinusFloat(Float=$System.Timeline.NormalizedTime))
       B=1
      )
     )
    )
    Scale RGB=&Multiply_Vector(
     A=&Multiply_VectorByFloat(
      Vector=&MakeVectorFromLinearColorRGB(Color=$System.TODColorMult)
      Float=&Multiply_Float(
       A=&Lerp_Float(
        A=0.5
        B=0.75
        Alpha=&FloatFromCurve(FloatCurve=DataInterface() {propertyValues {Curve {keys [
             {interpMode=RCIM_Cubic, tangentMode=RCTM_User, time=0, value=0}
             {interpMode=RCIM_Cubic, tangentMode=RCTM_User, time=0.1, value=1}
             {interpMode=RCIM_Cubic, tangentMode=RCTM_User, time=1, value=0}
            ]}}})
       )
       B=$User.NightMultiplier
      )
     )
     B=&MakeVectorFromLinearColorRGB(Color=&FNTimeOfDayColor(Night Color=<rgba 0.2604 0.998 2 1>, Morning Color=<rgba 1 1 1 1>, Day Color=<rgba 1.1347 2 1.5141 1>, Evening Color=<rgba 1 1 1 1>))
    )
    ScaleA=true
    ScaleRGB=true
   )
   &SolveForcesAndVelocity(Clamp Velocity=false, Limit Acceleration=false)
   &DynamicMaterialParameters(
    Index 0 Param 1=&FloatFromCurve(FloatCurve=DataInterface() {propertyValues {Curve {keys [
         {arriveTangent=1.526, interpMode=RCIM_Cubic, leaveTangent=1.526, tangentMode=RCTM_User, time=0, value=0}
         {interpMode=RCIM_Cubic, tangentMode=RCTM_User, time=1, value=1}
        ]}}})
    Module.Index0Param0Write=true
    Module.Index0Param1Write=true
    Module.Index0Param2Write=false
    Module.Index0Param3Write=true
   )
  ]
  rendererClasses [&NiagaraMeshRendererProperties]
  renderers [{rendererClass=&NiagaraMeshRendererProperties, rendererIndex=0}]
 }
}
summary {
 systemName=NS_MedCloud_Grenade_Dome
 emitters [
  {emitterName=Dome, rendererClasses [&NiagaraMeshRendererProperties]}
  {emitterName=Mist, rendererClasses [&NiagaraSpriteRendererProperties]}
  {emitterName=Motes, rendererClasses [&NiagaraSpriteRendererProperties]}
  {emitterName=WindGusts, rendererClasses [&NiagaraMeshRendererProperties]}
 ]
 userVariables [
  User.FadeOut {type=NiagaraBool}
  User.FX_Scale {type=NiagaraFloat, defaultValue=1.2, description=""}
  User.FXAge {type=NiagaraFloat, defaultValue=20, description=""}
  User.NightMultiplier {type=NiagaraFloat, defaultValue=1, description=""}
  User.SmokeColor {type=LinearColor, description="", defaultValue {struct=LinearColor, value=<rgba 0 1 0.4471 1>}}
 ]
}
