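# Niagara burst effect for a smoke grenade's cloud forming, with size mostly driven by FX_Scale.
# Mist throws out smoke sprites slowed by drag and gravity, tinted by time of day and offset against owner motion.
# Streaks add radial sprites, WindGusts adds rotating, expanding meshes, and StarGlow combines flash sprites, a dome mesh and a light tinted by SmokeColor.
# NightMultiplier scales gust and glow brightness; the GPU Motes emitter is disabled.
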NiagaraSystem /MedCloud/Assets/FX/NS_SmokeCloud_Grenade_Form
dependencies {
 usedDataInterfaces [
  &NiagaraDataInterfaceCurve
  &NiagaraDataInterfaceMeshRendererInfo
  &NiagaraDataInterfaceArrayFloat
  &NiagaraDataInterfaceVector2DCurve
  &NiagaraDataInterfaceColorCurve
 ]
 usedDynamicInputs []
 usedModules [
  &self:SystemSpawnScript.SystemScriptSource.SystemScriptGraph.NiagaraNodeAssignment_1.SetVariables_#1
  &SystemState
 ]
}
emitters {
 Mist {
  emitterUpdateScript [
   &EmitterState()
   &SpawnBurst_Instantaneous(Spawn Count=30, Spawn Time=0.03, Use Spawn Probability=false)
  ]
  particleSpawnScript [
   &InitializeParticle(
    Lifetime Max=2.5
    Lifetime Min=2
    Lifetime Mode=::NewEnumerator1
    Sprite Rotation Mode=::NewEnumerator2
    Sprite Size Mode=::NewEnumerator3
    Uniform Sprite Size=&Multiply_Float(A=&RandomRangeFloat(Minimum=650, Maximum=750), B=$User.FX_Scale)
    UsePositionOffset=false
    Write Material Random=true
   )
   /SetVariables(Particles.InitSpriteRotRate=&UniformRangedFloat(Minimum=-3, Maximum=3)) {isSetParameters=true}
   &ShapeLocation(
    MaskSpawnGroup=false
    Offset=<0 0 35>
    Offset Mode=::NewEnumerator0
    Sphere Distribution=::NewEnumerator1
    Sphere Radius=50
    U Position=&ReturnNormalizedExecIndex()
    UseOwnerScale=true
    V Position=&ReturnNormalizedExecIndex(Normalized Index Scale=6)
   )
   /SetVariables(Particles.Position=&MaskVectorBySpawnGroup(
     Masked Vector=&Add_Vector(A=Linked() {linkedVariable {name=Engine.Owner.Position, type=NiagaraPosition}}, B=<0 0 0>)
     Passthrough Vector=Linked() {linkedVariable {name=Particles.Position, type=NiagaraPosition}}
     Spawn Group Mask=1
    )) {isSetParameters=true}
   &AddVelocity(
    Module.Use Falloff Distance=false
    Origin Offset=&UniformRangedVector(Minimum=<0 0 0>, Maximum=<0 0 0>)
    Override Seed=false
    Velocity Mode=::NewEnumerator1
    Velocity Speed=&Multiply_Float(
     A=&Lerp_Float(
      A=900
      B=1000
      Alpha=&FloatFromCurve(
       FloatCurve=DataInterface() {propertyValues {Curve {keys [
           {interpMode=RCIM_Cubic, tangentMode=RCTM_Break, time=0, value=0}
           {arriveTangent=1.9418, interpMode=RCIM_Cubic, leaveTangent=1.9418, tangentMode=RCTM_User, time=0.3, value=0.55}
           {arriveTangentWeight=0.3338, interpMode=RCIM_Cubic, tangentMode=RCTM_Break, time=1, value=1}
          ]}}}
       CurveIndex=&RandomRangeFloat(Minimum=0, Maximum=1)
      )
     )
     B=$User.FX_Scale
    )
   ) {module=AddVelocity001}
   &DynamicMaterialParameters(
    Index 0 Param 1=2
    Index 0 Param 2=0.1
    Index 0 Param 3=1
    Module.Param0WriteEnabled=true
    Module.Param1WriteEnabled=true
    Module.Param2WriteEnabled=true
    Module.Param3WriteEnabled=true
   ) {module=DynamicMaterialParameters001}
  ]
  particleUpdateScript [
   &ParticleState()
   &SubUVAnimation()
   &ScaleColor(
    Scale Alpha=&FloatFromCurve(
     FloatCurve=DataInterface() {propertyValues {Curve {keys [
         {arriveTangent=12.535, interpMode=RCIM_Cubic, leaveTangent=12.535, tangentMode=RCTM_User, time=0, value=0}
         <at 0.2 1 RCIM_Cubic>
         {interpMode=RCIM_Cubic, tangentMode=RCTM_User, time=1, value=0}
        ]}}}
     Scale Curve=0.8
    )
    Scale RGB=&MakeVectorFromLinearColorRGB(Color=&MultiplyLinearColor(
      Linear Color=&FNTimeOfDayColor(
       Night Color=<rgba 0.5651 0.4284 0.3183 1>
       Morning Color=<rgba 0.917 1 0.75 1>
       Day Color=<rgba 0.7066 0.6753 0.5238 1>
       Evening Color=<rgba 0.5955 0.4791 0.3757 1>
      )
      Scale Factor=&MakeVector4_FromLinearColor(LinearColor=<rgba 1 1 1 1>)
     ))
    ScaleA=true
    ScaleRGB=true
   )
   &Drag(
    Drag=&FloatFromCurve(
     FloatCurve=DataInterface() {propertyValues {Curve {keys [
         {interpMode=RCIM_Cubic, leaveTangentWeight=0.4822, tangentMode=RCTM_Break, time=0, value=0.5}
         {interpMode=RCIM_Cubic, tangentMode=RCTM_Break, time=1, value=1}
        ]}}}
     Scale Curve=6
    )
    Use Linear Drag=true
    Use Rotational Drag=false
   )
   &GravityForce(Gravity=&RandomRangeVector(Minimum=<0 0 -100>, Maximum=<0 0 -700>))
   &InheritSourceMovement(
    Applied Position Scale=&VectorFromFloat(Value=&FloatFromCurve(FloatCurve=DataInterface() {propertyValues {Curve {keys [
          {interpMode=RCIM_Cubic, tangentMode=RCTM_User, time=0, value=0.5}
          {arriveTangent=-0.4702, interpMode=RCIM_Cubic, leaveTangent=-0.4702, tangentMode=RCTM_User, time=1, value=0}
         ]}}}))
    ApplyForce=false
    ApplyPosition=true
    ApplyPositionFalloff=false
    Source Velocity:Vector3f=&TransformVector(
     Vector=/Niagara/DynamicInputs/Math/Clamp_Vector(
      Vector:Vector3f=&Multiply_VectorByFloat(Vector=$Engine.Owner.Velocity, Float=-1)
      Min:Vector3f=<0 0 0>
      Max:Vector3f=<500 500 500>
     )
     Source Space=::World
     Destination Space=::Local
    )
   )
   &SolveForcesAndVelocity(Clamp Velocity=false, Limit Acceleration=false)
   &ScaleSpriteSize(Scale Factor=&Vector2DFromFloat(Value=&FloatFromCurve(
      FloatCurve=DataInterface() {propertyValues {Curve {keys [
          {arriveTangent=2.5906, interpMode=RCIM_Cubic, leaveTangent=2.5906, tangentMode=RCTM_User, time=0, value=0.3}
          {arriveTangent=0.1487, interpMode=RCIM_Cubic, leaveTangent=0.1487, tangentMode=RCTM_User, time=1, value=1}
         ]}}}
      Scale Curve=0.6
     )))
   &CameraOffset(Camera Offset Amount=80)
   &DynamicMaterialParameters(Module.Index0Param0Write=true, Module.Index0Param1Write=true, Module.Index0Param2Write=true, Module.Index0Param3Write=true)
  ]
  rendererClasses [&NiagaraSpriteRendererProperties]
  renderers [{rendererClass=&NiagaraSpriteRendererProperties}]
 }
 Motes {
  bEnabled=false
  simTarget=GPUComputeSim
  emitterUpdateScript [
   &EmitterState()
   &SpawnBurst_Instantaneous(Spawn Count=30, Use Spawn Probability=false, UseLoopCountLimit=false)
   &SpawnBurst_Instantaneous(Spawn Count=50, Spawn Group=1, Use Spawn Probability=false, UseLoopCountLimit=false) {module=SpawnBurst_Instantaneous001}
  ]
  particleSpawnScript [
   /SetVariables(
    Particles.Particle_Depth:NiagaraFloat=&FloatFromCurve(
     FloatCurve=DataInterface() {propertyValues {Curve {keys [
         {arriveTangent=1.2753, interpMode=RCIM_Cubic, leaveTangent=1.2753, tangentMode=RCTM_User, time=0, value=0}
         {arriveTangent=0.204, interpMode=RCIM_Cubic, leaveTangent=0.204, tangentMode=RCTM_User, time=0.9, value=1}
        ]}}}
     CurveIndex=&RandomRangeFloat(Minimum=0, Maximum=1)
    )
    Particles.Particle_Height:NiagaraFloat=&FloatFromCurve(
     FloatCurve=DataInterface() {propertyValues {Curve {keys [
         {arriveTangent=1.9477, interpMode=RCIM_Cubic, leaveTangent=1.9477, tangentMode=RCTM_User, time=0, value=0}
         {arriveTangent=0.1423, interpMode=RCIM_Cubic, leaveTangent=0.1423, tangentMode=RCTM_User, time=0.5, value=0.5}
         {arriveTangent=1.4427, interpMode=RCIM_Cubic, leaveTangent=1.4427, tangentMode=RCTM_User, time=1, value=1}
        ]}}}
     CurveIndex=&RandomRangeFloat(Minimum=0, Maximum=1)
    )
    Particles.Particle_Scale:NiagaraFloat=&RandomRangeFloat(Minimum=0, Maximum=1)
   ) {isSetParameters=true}
   &PartitionParticles()
   /SetVariables(Particles.VisibilityTag=&SpawnGroup()) {isSetParameters=true}
   &InitializeParticle(
    Color Maximum=<rgba 0 1 0.3727 1>
    Color Minimum=<rgba 0 0.5833 0.2061 1>
    Color Mode=::NewEnumerator2
    Lifetime Max=0.75
    Lifetime Min=0.25
    Lifetime Mode=::NewEnumerator1
    Mass Mode=::NewEnumerator3
    Sprite Size Mode=::NewEnumerator3
    Uniform Sprite Size=&Multiply_Float(A=$User.FX_Scale, B=&Lerp_Float(A=12, B=27, Alpha=$Particles.Particle_Scale))
    UsePositionOffset=false
    Write Material Random=true
   )
   &ShapeLocation(
    MaskSpawnGroup=false
    Override Seed=false
    Sphere Radius=&Multiply_Float(A=500, B=$User.FX_Scale)
    Sphere Surface Distribution=0.381
    UseOwnerScale=true
   )
   &CameraOffset(Camera Offset Amount=6)
   /SetVariables(Particles.RotationDirection:Vector3f=&Lerp_Vector(
     A=&RandomRangeVector(Minimum=<-0.6 -0.6 1>, Maximum=<0.8 0.8 1>)
     B=&RandomRangeVector(Minimum=<-0.2 -0.2 1>, Maximum=<0.3 0.3 1>)
     Alpha=&FloatFromCurve(
      FloatCurve=DataInterface() {propertyValues {Curve {keys [
          {interpMode=RCIM_Cubic, tangentMode=RCTM_User, time=0, value=0}
          {arriveTangent=0.5342, interpMode=RCIM_Cubic, leaveTangent=0.5342, tangentMode=RCTM_User, time=1, value=1}
         ]}}}
      CurveIndex=$Particles.Particle_Depth
     )
    )) {isSetParameters=true}
   &DynamicMaterialParameters(
    Index 0 Param 1=&RandomRangeFloat(Minimum=1, Maximum=2)
    Module.Index0Param0Write=true
    Module.Index0Param1Write=false
    Module.Index0Param2Write=false
    Module.Index0Param3Write=false
   )
   &AddVelocity(Module.Use Falloff Distance=false, Override Seed=false, Velocity Mode=::NewEnumerator1, Velocity Speed=8000)
  ]
  particleUpdateScript [
   &ParticleState()
   &ScaleSpriteSize(Uniform Curve Sprite Scale=DataInterface() {propertyValues {Curve {keys [
        {arriveTangent=3.8184, interpMode=RCIM_Cubic, leaveTangent=3.8184, tangentMode=RCTM_User, time=0, value=0.6}
        {arriveTangentWeight=0.0359, interpMode=RCIM_Cubic, leaveTangentWeight=0.3231, tangentMode=RCTM_User, time=0.2, value=1}
        {arriveTangent=-3.3427, interpMode=RCIM_Cubic, leaveTangent=-3.3427, tangentMode=RCTM_User, time=1, value=0.2}
       ]}}})
   &RotateAroundPoint(
    Derive the Rotation Axis=false
    Radius:NiagaraFloat=&RandomRangeFloat(Minimum=300, Maximum=500)
    Rotation Phase:NiagaraFloat=&RandomRangeFloat(Minimum=-1, Maximum=1)
    Rotation Rate:NiagaraFloat=&Lerp_Float(
     A=&RandomRangeFloat(Minimum=0.2, Maximum=1)
     B=0.1
     Alpha=&FloatFromCurve(FloatCurve=DataInterface() {propertyValues {Curve {keys [
          {arriveTangent=1.5185, interpMode=RCIM_Cubic, leaveTangent=1.5185, tangentMode=RCTM_User, time=0, value=0}
          {arriveTangentWeight=0.7852, interpMode=RCIM_Cubic, tangentMode=RCTM_User, time=1, value=1}
         ]}}})
    )
   )
   &AccelerationForce(Acceleration=&RandomRangeVector(Minimum=<0 0 1000>, Maximum=<0 0 15000>))
   &SolveForcesAndVelocity(Clamp Velocity=false, Limit Acceleration=false)
   &RotateAroundPoint(
    Derive the Rotation Axis=false
    Radius:NiagaraFloat=&Multiply_Float(
     A=$User.FX_Scale
     B=&Lerp_Float(
      A=450
      B=&Multiply_Float(
       A=&RandomRangeFloat(Minimum=0.9, Maximum=1)
       B=&Lerp_Float(
        A=50
        B=150
        Alpha=&FloatFromCurve(
         FloatCurve=DataInterface() {propertyValues {Curve {keys [
             {arriveTangent=5.5061, interpMode=RCIM_Cubic, leaveTangent=5.5061, tangentMode=RCTM_User, time=0, value=0}
             {interpMode=RCIM_Cubic, tangentMode=RCTM_User, time=0.5, value=1}
             {arriveTangent=-6.0544, interpMode=RCIM_Cubic, leaveTangent=-6.0544, tangentMode=RCTM_User, time=1, value=0}
            ]}}}
         CurveIndex=$Particles.Particle_Height
        )
       )
      )
      Alpha=$Particles.Particle_Depth
     )
    )
    Rotation Axis:Vector3f=$Particles.RotationDirection
    Rotation Center:NiagaraPosition=&AddVectorToPosition(
     Position=$Engine.Owner.Position
     Vector=&Multiply_VectorByFloat(
      Vector=$Particles.RotationDirection
      Float=&Multiply_Float(
       A=&Multiply_Float(A=$User.FX_Scale, B=&Lerp_Float(A=-100, B=100, Alpha=$Particles.Particle_Height))
       B=&Lerp_Float(A=0.5, B=1, Alpha=$Particles.Particle_Depth)
      )
     )
    )
    Rotation Phase:NiagaraFloat=&RandomRangeFloat(Minimum=-1, Maximum=1)
    Rotation Rate:NiagaraFloat=&Multiply_Float(
     A=&FloatFromCurve(FloatCurve=DataInterface() {propertyValues {Curve {keys [
          {interpMode=RCIM_Cubic, tangentMode=RCTM_User, time=0, value=0.2}
          {arriveTangent=0.3, interpMode=RCIM_Cubic, leaveTangent=0.3, time=0.3, value=1}
          {interpMode=RCIM_Cubic, tangentMode=RCTM_User, time=1, value=0.5}
         ]}}})
     B=&Multiply_Float(
      A=&Lerp_Float(A=1, B=0.7, Alpha=$Particles.Particle_Scale)
      B=&Multiply_Float(
       A=&Lerp_Float(A=2.5, B=0.5, Alpha=$Particles.Particle_Depth)
       B=&Lerp_Float(
        A=1
        B=1
        Alpha=&FloatFromCurve(
         FloatCurve=DataInterface() {propertyValues {Curve {keys [
             {arriveTangent=5.5061, interpMode=RCIM_Cubic, leaveTangent=5.5061, tangentMode=RCTM_User, time=0, value=0}
             {interpMode=RCIM_Cubic, tangentMode=RCTM_User, time=0.5, value=1}
             {arriveTangent=-6.0544, interpMode=RCIM_Cubic, leaveTangent=-6.0544, tangentMode=RCTM_User, time=1, value=0}
            ]}}}
         CurveIndex=$Particles.Particle_Height
        )
       )
      )
     )
    )
   ) {enabled=false, module=RotateAroundPoint001}
   &ScaleColor(Scale RGB=&VectorFromFloat(Value=5), ScaleA=true, ScaleRGB=true)
  ]
  rendererClasses [&NiagaraSpriteRendererProperties]
  renderers [
   {rendererClass=&NiagaraSpriteRendererProperties}
   {rendererClass=&NiagaraSpriteRendererProperties, rendererIndex=1}
  ]
 }
 NE_StarGlow {
  emitterUpdateScript [
   &EmitterState()
   &SpawnBurst_Instantaneous(Use Spawn Probability=false, UseLoopCountLimit=false)
  ]
  particleSpawnScript [
   &InitializeParticle(
    Lifetime=0.6
    Mass Mode=::NewEnumerator3
    Mesh Scale Mode=::NewEnumerator3
    Mesh Uniform Scale=&Multiply_Float(A=10, B=$User.FX_Scale)
    Position Offset=<0 0 50>
    Sprite Size Mode=::NewEnumerator3
    Uniform Sprite Size=&Multiply_Float(A=1800, B=$User.FX_Scale)
    UsePositionOffset=true
    Write Material Random=true
   )
   &CameraOffset(Camera Offset Amount=4)
  ]
  particleUpdateScript [
   &ParticleState()
   &SpriteRotationRate(Rotation Rate=&FloatFromCurve(
     FloatCurve=DataInterface() {propertyValues {Curve {keys [
         {arriveTangent=1.4306, interpMode=RCIM_Cubic, leaveTangent=1.4306, tangentMode=RCTM_User, time=0.2, value=0}
         {arriveTangent=1.1179, interpMode=RCIM_Cubic, leaveTangent=1.1179, tangentMode=RCTM_User, time=0.4, value=0.6228}
         {arriveTangent=0.0224, interpMode=RCIM_Cubic, leaveTangent=0.0224, tangentMode=RCTM_User, time=1, value=1}
        ]}}}
     Scale Curve=10
    ))
   &ScaleColor(
    Scale Alpha=&FloatFromCurve(FloatCurve=DataInterface() {propertyValues {Curve {keys [<at 0 1>,<at 1 0>]}}})
    Scale RGB=&Multiply_Vector(
     A=&Multiply_VectorByFloat(
      Vector=&MakeVectorFromLinearColorRGB(Color=&ColorFromCurve(ColorCurve=DataInterface() {propertyValues {
          AlphaCurve {keys [<at 0 1 RCIM_Cubic>]}
          BlueCurve {keys [<at 0.0211 5>, {interpMode=RCIM_Cubic, leaveTangent=1.6293, time=0.3406, value=1.5}]}
          GreenCurve {keys [<at 0.0211 5>, {interpMode=RCIM_Cubic, leaveTangent=-30.3699, time=0.3406, value=1.5}]}
          RedCurve {keys [<at 0.0211 5>, {interpMode=RCIM_Cubic, leaveTangent=-80.5167, time=0.3406, value=1.5}]}
         }}))
      Float=$User.NightMultiplier
     )
     B=&MakeVectorFromLinearColorRGB(Color=&FNTimeOfDayColor(Night Color=<rgba 1 1 1 1>, Morning Color=<rgba 0 0.5388 1 1>, Day Color=<rgba 1 1 1 1>, Evening Color=<rgba 0 0.5388 1 1>))
    )
    ScaleA=true
    ScaleRGB=true
   ) {module=ScaleColor001}
   &ScaleSpriteSize(
    Scale Sprite Size Mode=::NewEnumerator0
    Uniform Scale Factor=&FloatFromCurve(FloatCurve=DataInterface() {propertyValues {Curve {keys [
         {arriveTangent=3.3646, interpMode=RCIM_Cubic, leaveTangent=3.3646, tangentMode=RCTM_User, time=0, value=0}
         {interpMode=RCIM_Cubic, tangentMode=RCTM_User, time=1, value=1}
        ]}}})
   )
   /SetVariables(Particles.DiamondSize:Vector2f=&Multiply_Vector2D(A=$Particles.SpriteSize, B=<1 0.25>)) {isSetParameters=true}
   /SetVariables(Particles.GlowColor=&ColorFromCurve(ColorCurve=DataInterface() {propertyValues {
       AlphaCurve {keys [<at 0 1 RCIM_Cubic>,<at 1 0 RCIM_Cubic>]}
       BlueCurve {keys [<at 0 0 RCIM_Cubic>]}
       GreenCurve {keys [<at 0 1.9201 RCIM_Cubic>]}
       RedCurve {keys [<at 0 10 RCIM_Cubic>]}
      }})) {isSetParameters=true}
   /SetVariables(
    Particles.DomeColor:LinearColor=&ColorFromCurve(ColorCurve=DataInterface() {propertyValues {
       AlphaCurve {keys [<at 0 1 RCIM_Cubic>,<at 1 0 RCIM_Cubic>]}
       BlueCurve {keys [<at 0 0 RCIM_Cubic>,<at 0.3844 1 RCIM_Cubic>]}
       GreenCurve {keys [<at 0 4.9507 RCIM_Cubic>,<at 0.3844 1 RCIM_Cubic>]}
       RedCurve {keys [<at 0 50 RCIM_Cubic>,<at 0.3844 1 RCIM_Cubic>]}
      }})
    Particles.DomeDynMat:Vector4f=&MakeVector4(
     X=1
     Y=6
     Z=&FloatFromCurve(FloatCurve=DataInterface() {propertyValues {Curve {keys [
          {arriveTangent=-0.9649, interpMode=RCIM_Cubic, leaveTangent=-0.9649, tangentMode=RCTM_User, time=0, value=1}
          {arriveTangent=-6.2582, leaveTangent=-1.25, tangentMode=RCTM_User, time=0.03, value=0}
          {arriveTangent=3.1809, interpMode=RCIM_Cubic, leaveTangent=3.1809, tangentMode=RCTM_User, time=0.8, value=0}
          {arriveTangent=1.435, interpMode=RCIM_Cubic, leaveTangent=1.435, tangentMode=RCTM_User, time=1, value=1}
         ]}}})
     W=&Multiply_Float(A=$User.FX_Scale, B=1)
    )
   ) {isSetParameters=true}
   &ScaleMeshSize(Scale Factor=&VectorFromFloat(Value=&Lerp_Float(
      A=0.3
      B=1.3
      Alpha=&FloatFromCurve(FloatCurve=DataInterface() {propertyValues {Curve {keys [
           {arriveTangent=1.6474, interpMode=RCIM_Cubic, leaveTangent=1.6474, tangentMode=RCTM_User, time=0, value=0}
           {arriveTangent=0.3409, interpMode=RCIM_Cubic, leaveTangent=0.3409, tangentMode=RCTM_User, time=1, value=1}
          ]}}})
     )))
   &FN_Light_Attributes(
    Color=&MultiplyLinearColor(
     Linear Color=&ColorFromCurve(ColorCurve=DataInterface() {propertyValues {
        AlphaCurve {keys [<at 0.821 1 RCIM_Cubic>,<at 1 0 RCIM_Cubic>]}
        BlueCurve {keys [<at 0.4282 4.8419 RCIM_Cubic>,<at 1 0.8812 RCIM_Cubic>]}
        GreenCurve {keys [<at 0.4282 4.9683 RCIM_Cubic>,<at 1 0.9694 RCIM_Cubic>]}
        RedCurve {keys [<at 0.4282 5 RCIM_Cubic>,<at 1 1 RCIM_Cubic>]}
       }})
     Scale Factor=&MakeVector4_FromLinearColor(LinearColor=$User.SmokeColor)
    )
    Create Color Attribute=true
    Create Exponent Attribute=true
    Create Position Attribute=true
    Create Radius Attribute=true
    Exponent=2
    Position=<0 0 200>
    Radius=750
   )
  ]
  rendererClasses [
   &NiagaraSpriteRendererProperties
   &NiagaraLightRendererProperties
   &NiagaraMeshRendererProperties
  ]
  renderers [
   {rendererClass=&NiagaraSpriteRendererProperties}
   {rendererClass=&NiagaraSpriteRendererProperties, rendererIndex=1}
   {rendererClass=&NiagaraLightRendererProperties, rendererIndex=2}
   {rendererClass=&NiagaraMeshRendererProperties, rendererIndex=3}
  ]
 }
 Streaks {
  emitterUpdateScript [
   &EmitterState()
   &SpawnBurst_Instantaneous(Spawn Count=12, Use Spawn Probability=false, UseLoopCountLimit=false)
   &SpawnBurst_Instantaneous(Spawn Count=20, Spawn Group=1, Use Spawn Probability=false, UseLoopCountLimit=false) {module=SpawnBurst_Instantaneous001}
  ]
  particleSpawnScript [
   /SetVariables(Particles.VisibilityTag=&SpawnGroup()) {isSetParameters=true}
   /SetVariables(Particles.RandomFloat=&UniformRangedFloat(Minimum=0, Maximum=3)) {isSetParameters=true}
   &InitializeParticle(
    Color=<rgba 2 2 2 1>
    Lifetime=&MaskFloatBySpawnGroup(
     Masked Float=&RandomRangeFloat(Minimum=0.3, Maximum=0.45)
     Passthrough Float=&RandomRangeFloat(Minimum=0.55, Maximum=0.65)
    )
    Sprite Size=&Multiply_Vector2D_ByFloat(Float=$User.FX_Scale, Vector2D=&RandomRangeVector2D(Minimum=<15 100>, Maximum=<20 150>))
    Sprite Size Mode=::NewEnumerator4
    UsePositionOffset=false
    Write Material Random=true
   ) {module=InitializeParticle001}
   &ShapeLocation(
    MaskSpawnGroup=false
    Sphere Distribution=::NewEnumerator1
    Sphere Radius=&Multiply_Float(A=60, B=$User.FX_Scale)
    U Position=&Lerp_Float(A=0, B=1, Alpha=&ReturnNormalizedExecIndex(Normalized Index Scale=2))
    UseOwnerScale=true
    V Position=&RandomRangeFloat(Minimum=0, Maximum=1)
   )
   &AddVelocity(
    Module.Use Falloff Distance=false
    Override Seed=false
    Velocity Mode=::NewEnumerator1
    Velocity Speed=&RandomRangeFloat(Minimum=100, Maximum=300)
   )
   &SubUVAnimation(End Frame=3, SubUV Animation Mode:ENiagara_SubUVLookupMode=::NewEnumerator2)
  ]
  particleUpdateScript [
   &ParticleState() {module=UpdateAge}
   &SolveForcesAndVelocity(Clamp Velocity=false, Limit Acceleration=false)
   &ScaleSpriteSize(Scale Factor=&Vector2DFromCurve(
     Vector2Curve=DataInterface() {propertyValues {XCurve {keys [<at 0.0036 1.6897>,<at 0.9967 0.0582>]}, YCurve {keys [<at 0 1>,<at 1.002 3.7227>]}}}
     ScaleCurve=<1 1.25>
    ))
   &ScaleSpriteSize(Scale Sprite Size Mode=::NewEnumerator0, Uniform Scale Factor=2) {module=ScaleSpriteSize001}
   &ScaleColor(
    Scale Alpha=&FloatFromCurve(
     FloatCurve=DataInterface() {propertyValues {Curve {keys [{interpMode=RCIM_Cubic, tangentMode=RCTM_Break, time=0, value=1}, {interpMode=RCIM_Cubic, tangentMode=RCTM_Break, time=1, value=0}]}}}
     Scale Curve=&RandomRangeFloat(Minimum=0.2, Maximum=0.5)
    )
    Scale RGB=&VectorFromFloat(Value=1)
    ScaleA=true
    ScaleRGB=true
   )
   &DynamicMaterialParameters(
    Index 0 Param 2=1
    Index 0 Param 3=1
    Index 0 Param 4=&FloatFromCurve(FloatCurve=DataInterface() {propertyValues {Curve {keys [<at 0.3 1>,<at 1 0.25>]}}})
    Module.Param0WriteEnabled=true
    Module.Param1WriteEnabled=true
    Module.Param2WriteEnabled=true
    Module.Param3WriteEnabled=true
   )
   &FN_Sprite_Attributes(
    Color=&ColorFromCurve(ColorCurve=DataInterface() {propertyValues {
       AlphaCurve {keys [<at 0 1 RCIM_Cubic>,<at 0.4865 0 RCIM_Cubic>]}
       BlueCurve {keys [<at 0 0 RCIM_Cubic>,<at 0.4817 1 RCIM_Cubic>]}
       GreenCurve {keys [<at 0 1.5397 RCIM_Cubic>,<at 0.4817 1 RCIM_Cubic>]}
       RedCurve {keys [<at 0 5 RCIM_Cubic>,<at 0.4817 1 RCIM_Cubic>]}
      }})
    Create Color Attribute=true
    Create SpriteSize Attribute=true
    SpriteSize=&Multiply_Vector2D(
     A=&Multiply_Vector2D_ByFloat(
      Float=&FloatFromCurve(FloatCurve=DataInterface() {propertyValues {Curve {keys [
           {arriveTangent=2.0201, interpMode=RCIM_Cubic, leaveTangent=2.0201, tangentMode=RCTM_User, time=0, value=0}
           {arriveTangent=-0.026, interpMode=RCIM_Cubic, leaveTangent=-0.026, tangentMode=RCTM_User, time=0.6, value=1}
          ]}}})
      Vector2D=&Multiply_Vector2D_ByFloat(Float=$User.FX_Scale, Vector2D=&RandomRangeVector2D(Minimum=<350 750>, Maximum=<450 1500>))
     )
     B=&Vector2DFromCurve(
      Vector2Curve=DataInterface() {propertyValues {
        XCurve {keys [
          {arriveTangent=-1.6611, interpMode=RCIM_Cubic, leaveTangent=-1.6611, tangentMode=RCTM_User, time=0.2, value=1.2}
          {arriveTangent=-0.1324, interpMode=RCIM_Cubic, leaveTangent=-0.1324, tangentMode=RCTM_User, time=1, value=0.7}
         ]}
        YCurve {keys [{interpMode=RCIM_Cubic, tangentMode=RCTM_User, time=0.4, value=1}]}
       }}
      ScaleCurve=<1 1>
     )
    )
    Write to Module Namespace=true
   )
  ]
  rendererClasses [&NiagaraSpriteRendererProperties]
  renderers [
   {rendererClass=&NiagaraSpriteRendererProperties}
   {rendererClass=&NiagaraSpriteRendererProperties, rendererIndex=1}
  ]
 }
 WindGusts {
  emitterUpdateScript [
   &EmitterState()
   &SpawnBurst_Instantaneous(Spawn Count=10, Use Spawn Probability=false, UseLoopCountLimit=false)
  ]
  particleSpawnScript [
   /SetVariables(Particles.Particle_Height:NiagaraFloat=&Lerp_Float(A=0.3, B=0.7, Alpha=&ReturnNormalizedExecIndex())) {isSetParameters=true}
   &InitializeParticle(
    Lifetime Max=1.5
    Lifetime Min=1.25
    Lifetime Mode=::NewEnumerator1
    Mass Mode=::NewEnumerator3
    Mesh Scale=&Multiply_VectorByFloat(
     Vector=&RandomRangeVector(Minimum=<1 1 0.5>, Maximum=<1 1 1.5>)
     Float=&Multiply_Float(
      A=$User.FX_Scale
      B=&Lerp_Float(
       A=7.3
       B=8.7
       Alpha=&FloatFromCurve(
        FloatCurve=DataInterface() {propertyValues {Curve {keys [
            {arriveTangent=5.7133, interpMode=RCIM_Cubic, leaveTangent=5.7133, tangentMode=RCTM_User, time=0, value=0}
            {interpMode=RCIM_Cubic, tangentMode=RCTM_User, time=0.5, value=1}
            {arriveTangent=-6.0152, interpMode=RCIM_Cubic, leaveTangent=-6.0152, tangentMode=RCTM_User, time=1, value=0}
           ]}}}
        CurveIndex=$Particles.Particle_Height
        Scale Curve=&RandomRangeFloat(Minimum=0.9, Maximum=1)
       )
      )
     )
    )
    Mesh Scale Mode=::NewEnumerator4
    Sprite Rotation Mode=::NewEnumerator2
    UsePositionOffset=false
    Write Material Random=true
   )
   &SystemLocation(
    MaskSpawnGroup=false
    Offset=&MakeVector(
     X=0
     Y=0
     Z=&Multiply_Float(A=$User.FX_Scale, B=&Lerp_Float(A=-300, B=300, Alpha=$Particles.Particle_Height))
    )
   )
   &InitialMeshOrientation(
    Mesh Orientation Mode=::NewEnumerator2
    Orientation Coordinate Space=::World
    Rotation=&AddVectorToPosition(
     Position=&RandomRangeVector(Minimum=<-0.04 -0.04 0>, Maximum=<0.04 0.04 0>)
     Vector=&Lerp_Vector(A=<0 0 -1>, B=<0 0 1>, Alpha=&ReturnNormalizedExecIndex())
    )
    Use Rotation Vector=true
   )
   &CameraOffset(Camera Offset Amount=15)
  ]
  particleUpdateScript [
   &ParticleState()
   &ScaleMeshSize(Scale Factor=&Multiply_VectorByFloat(
     Vector=&MakeVector(
      X=1
      Y=1
      Z=&SelectFloatFromArray(
       Array Sampling Mode=::NewEnumerator0
       Direct Array Index=?
       Float Selection Array=DataInterface() {propertyValues {FloatData [1,-1]}}
       Random Seed=0
      )
     )
     Float=&Lerp_Float(
      A=0.2
      B=1.3
      Alpha=&FloatFromCurve(FloatCurve=DataInterface() {propertyValues {Curve {keys [
           {arriveTangent=3.9971, interpMode=RCIM_Cubic, leaveTangent=3.9971, tangentMode=RCTM_User, time=0, value=0}
           {interpMode=RCIM_Cubic, tangentMode=RCTM_User, time=0.86, value=1}
          ]}}})
     )
    ))
   &UpdateMeshOrientation(Rotation Rate=&RandomRangeFloat(Minimum=0.5, Maximum=0.6), Rotation Vector=<0 0 1>)
   &ScaleColor(
    Scale Alpha=&FloatFromCurve(
     FloatCurve=DataInterface() {propertyValues {Curve {keys [
         {arriveTangent=6.2301, interpMode=RCIM_Cubic, leaveTangent=6.2301, tangentMode=RCTM_User, time=0, value=0}
         {arriveTangent=-6.912, interpMode=RCIM_Cubic, leaveTangent=-6.912, tangentMode=RCTM_User, time=0.1, value=3}
         {arriveTangent=0.4048, interpMode=RCIM_Cubic, leaveTangent=0.4048, tangentMode=RCTM_User, time=0.7, value=0}
        ]}}}
     Scale Curve=&RandomRangeFloat(Minimum=0.8, Maximum=2)
    )
    Scale RGB=&VectorFromFloat(Value=&Multiply_Float(
      A=&Lerp_Float(
       A=0.5
       B=0.75
       Alpha=&FloatFromCurve(FloatCurve=DataInterface() {propertyValues {Curve {keys [
            {interpMode=RCIM_Cubic, tangentMode=RCTM_User, time=0, value=0}
            {interpMode=RCIM_Cubic, tangentMode=RCTM_User, time=0.1, value=1}
            {interpMode=RCIM_Cubic, tangentMode=RCTM_User, time=1, value=0}
           ]}}})
      )
      B=$User.NightMultiplier
     ))
    ScaleA=true
    ScaleRGB=true
   )
   &SolveForcesAndVelocity(Clamp Velocity=false, Limit Acceleration=false)
   &DynamicMaterialParameters(
    Index 0 Param 1=&FloatFromCurve(FloatCurve=DataInterface() {propertyValues {Curve {keys [
         {arriveTangent=1.526, interpMode=RCIM_Cubic, leaveTangent=1.526, tangentMode=RCTM_User, time=0, value=0}
         {interpMode=RCIM_Cubic, tangentMode=RCTM_User, time=1, value=1}
        ]}}})
    Module.Index0Param0Write=true
    Module.Index0Param1Write=true
    Module.Index0Param2Write=false
    Module.Index0Param3Write=true
   )
  ]
  rendererClasses [&NiagaraMeshRendererProperties]
  renderers [{rendererClass=&NiagaraMeshRendererProperties, rendererIndex=0}]
 }
}
summary {
 systemName=NS_SmokeCloud_Grenade_Form
 emitters [
  {bEnabled=false, emitterName=Motes, simTarget=GPUComputeSim, rendererClasses [&NiagaraSpriteRendererProperties]}
  {emitterName=WindGusts, rendererClasses [&NiagaraMeshRendererProperties]}
  {emitterName=Mist, rendererClasses [&NiagaraSpriteRendererProperties]}
  {emitterName=Streaks, rendererClasses [&NiagaraSpriteRendererProperties]}
  {
   emitterName=NE_StarGlow
   rendererClasses [
    &NiagaraSpriteRendererProperties
    &NiagaraLightRendererProperties
    &NiagaraMeshRendererProperties
   ]
  }
 ]
 userVariables [
  User.Bounds {type=Vector3f, description="", defaultValue {struct=Vector3f, value=<0 0 0>}}
  User.FX_Scale {type=NiagaraFloat, defaultValue=1, description=""}
  User.NightMultiplier {type=NiagaraFloat, defaultValue=1, description=""}
  User.SmokeColor {type=LinearColor, description="", defaultValue {struct=LinearColor, value=<rgba 0.5885 0.5885 0.5885 1>}}
 ]
}
