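# Masked, lit material for the MedMist pickup mesh, drawing a drifting two-tone green cloud.
# Stabilized screen-space UVs, corrected for object scale, feed panned cloud textures with Perlin distortion and bump-offset parallax.
# The cloud sample blends exposure-compensated colors and yields screen-derivative normals.
# View-angle falloff times panned, contrast-adjusted noise, smoothstepped against the cloud, drives dithered temporal-AA soft opacity.
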Material /MedCloud/Assets/Meshes/M_MedMist_Pickup_Material
graph {
 properties {
  bAutomaticallySetUsageInEditor=false
  bTangentSpaceNormal=false
  bUsedWithMeshDeformer=true
  bUsedWithSkeletalMesh=true
  blendMode=BLEND_Masked
  usedShadingModels=Default Lit
 }
 nodes [
  n1 = MaterialFunctionCall!(ObjectPivotLocation, ObjectPositionOffset=Constant3Vector()) {materialFunction=/Game/Characters/Player/Common/MaterialFunctions/Superlevel/ImprovedStabilizedScreenSpaceUVs}
  n2 = MaterialFunctionCall() {materialFunction=&ObjectScale}
  n3 = AppendVector(A=@n2.Scale X, B=@n2.Scale X)
  n4 = Multiply(A=Multiply(A=Divide(A=@n1.StabilizedUVs, B=@n3)) {constB=0.1}, B=ScalarParameter() {defaultValue=50, parameterName=Scale}) {constB=2}
  n5 = Constant2Vector() {g=0.05, r=0.01}
  n6 = Panner(Coordinate=@n4, Speed=@n5) {speedX=0.01, speedY=0.05}
  n7 = TextureSample(UVs=@n6) {samplerSource=SSM_Wrap_WorldGroupSettings, texture=/Game/Athena/SafeZone/S11/T_TilingClouds_D}
  n8 = Panner(Coordinate=@n6, Speed=Multiply(A=@n5, B=Constant2Vector() {g=2, r=0.1})) {speedX=0.01, speedY=0.05}
  n9 = TextureSample(UVs=@n8) {samplerSource=SSM_Wrap_WorldGroupSettings, texture=&T_TilingClouds_D}
  InverseLinearInterpolate(B=@n7.G, Value=@n9.G)
  MaterialFunctionCall!(MA=GetMaterialAttributes().MaterialAttributes, SkinFX=Constant() {r=1}) {materialFunction=/Game/Global/Materials/MaterialGym/Materials/Nitro/MF_Nitro_v2}
  n10 = TextureSample(UVs=@n8) {
   samplerSource=SSM_Wrap_WorldGroupSettings
   samplerType=SAMPLERTYPE_Masks
   texture=/Game/Characters/Player/Female/Medium/Bodies/F_MED_Neon_Tight_Suit/Materials/Soft_Perlin_output
  }
  n11 = Multiply(A=Subtract(A=@n10.G) {constB=0.5}) {constB=0.1}
  n12 = Add(A=@n6, B=@n11)
  n13 = TextureSample(UVs=BumpOffset(
    Coordinate=@n12
    Height=TextureSample(UVs=@n12) {samplerSource=SSM_Wrap_WorldGroupSettings, texture=&T_TilingClouds_D}.G
   ) {heightRatio=0.35}) {samplerSource=SSM_Wrap_WorldGroupSettings, texture=&T_TilingClouds_D}
  n14 = Multiply(A=AppendVector(A=DDX(Value=@n13.G), B=DDY(Value=@n13.G))) {constB=-1}
  n15 = Constant() {r=1}
  MaterialFunctionCall(Normal=Normalize(VectorInput=AppendVector(A=@n14, B=@n15)), Flatness=Constant() {r=-5}) {materialFunction=&FlattenNormal}
  n16 = VectorParameter() {defaultValue=<rgba 0.0154 0.1198 0.0751 1>}
  n17 = VectorParameter() {parameterName=Param_1, defaultValue=<rgba 0 1 0.4078 1>}
  n18 = LinearInterpolate(Alpha=@n13.G) {constA=0.15}
  Multiply(A=Multiply(A=LinearInterpolate(A=@n16.RGB, B=@n17.RGB, Alpha=@n18)) {constB=0.5}, B=EyeAdaptationInverse().EyeAdaptationInverse)
  Multiply(A=@n12)
  Multiply(A=@n4, B=Constant2Vector() {g=0.5, r=1})
  Multiply(A=@n1.StabilizedUVs, B=@n3)
  n19 = Multiply(A=Normalize(VectorInput=WorldPosition() {worldPositionShaderOffset=WPT_CameraRelative}.XYZ)) {constB=-1}
  n20 = VertexNormalWS()
  n21 = Constant3Vector() {constant=<rgba 0 0 1 0>}
  n22 = DotProduct(
   A=CameraVectorWS()
   B=Divide(
    A=Transform(None=@n21) {transformSourceType=TRANSFORMSOURCE_Local}
    B=MaterialFunctionCall() {materialFunction=&ObjectScale}.Scale XYZ
   )
  )
  n23 = LinearInterpolate(A=Power(Base=DotProduct(A=@n19, B=@n20)) {constExponent=0.5}, Alpha=Power(Base=Abs(None=@n22)) {constExponent=1}) {constA=1}
  n24 = Multiply(A=@n23, B=@n23)
  n25 = Panner(Coordinate=@n6) {speedX=0.25}
  n26 = ScalarParameter() {defaultValue=0.25, parameterName=B}
  n27 = TextureSample(UVs=Subtract(A=Multiply(A=@n25, B=@n26) {constB=0.25}, B=@n11)) {
   samplerSource=SSM_Wrap_WorldGroupSettings
   samplerType=SAMPLERTYPE_Masks
   texture=&Soft_Perlin_output
  }
  n28 = MaterialFunctionCall(In=Power(Base=@n27.G, Exp=ScalarParameter() {defaultValue=0.25, parameterName=Exp}) {constExponent=1}, Contrast=Constant()) {materialFunction=&CheapContrast}
  n29 = Multiply(A=SmoothStep(Min=@n13.G, Value=Multiply(A=@n24, B=@n28.Result))) {constB=2}
  Reroute(None=MaterialFunctionCall(Alpha Threshold=Saturate(None=@n29)) {materialFunction=&DitherTemporalAA}.Result)
  SmoothStep(Max=@n7.G, Value=@n9.G)
  SmoothStep(Value=@n28.Result)
  Subtract(A=@n7.G, B=@n9.G)
  TextureCoordinate()
  TextureSample() {samplerSource=SSM_Wrap_WorldGroupSettings, samplerType=SAMPLERTYPE_Masks, texture=/FortWater/Textures/OceanNoise__Mask}
 ]
}
